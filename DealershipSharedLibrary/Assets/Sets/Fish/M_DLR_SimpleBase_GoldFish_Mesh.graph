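# Masked, lit fish material for Niagara mesh particles, named for a goldfish, with particle-driven color, motion and dissolve.
# World position offset math blends a time-based sine/cosine flop with a UV-gradient swim rotation via UseSwimWPO, both around the object pivot.
# Dynamic parameters drive flop speed and amount, swim axis, Fresnel glow and a noise-based dithered dissolve.
# Base color combines clamped, particle-tinted diffuse with a masked eye texture.
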Material /DealershipSharedLibrary/Assets/Sets/Fish/M_DLR_SimpleBase_GoldFish_Mesh
graph {
 properties {bAutomaticallySetUsageInEditor=false, bUsedWithNiagaraMeshParticles=true, blendMode=BLEND_Masked, usedShadingModels=Default Lit}
 nodes [
  Constant()
  n1 = Multiply(A=NamedRerouteUsage(), B=Time())
  n2 = Multiply(A=NamedRerouteUsage(), B=AppendVector(A=Sine(None=@n1), B=Cosine(None=@n1)))
  n3 = Constant() {r=0.02}
  n4 = Add(A=ComponentMask(None=TextureCoordinate()) {r=true}, B=Clamp(None=TextureCoordinate() {coordinateIndex=1}) {maxDefault=0.2})
  n5 = WorldPosition()
  n6 = MaterialFunctionCall() {materialFunction=&ObjectPivotPoint}
  n7 = DynamicParameter() {parameterIndex=1, defaultValue=<rgba 0 0 0 1>, paramNames [RotationX,RotationY,RotationZ,Param4]}
  n8 = MaterialFunctionCall(R=@n7.RotationX, G=@n7.RotationX, B=@n7.RotationX) {materialFunction=&AppendMany}
  n9 = TextureCoordinate()
  n10 = Power(
   Base=Multiply(
    A=ChannelMaskParameter(None=@n9) {parameterName=SwimUVGradientDirection}.R
    B=ScalarParameter() {defaultValue=0.4, parameterName=SwimMaskStrength}
   )
   Exp=ScalarParameter() {defaultValue=3, parameterName=SwimMaskPower}
  ) {constExponent=10}
  LinearInterpolate(
   A=RotateAboutAxis(NormalizedRotationAxis=@n2, RotationAngle=Multiply(A=@n3, B=@n4), PivotPoint=@n6.Object Pivot Location, Position=@n5.XYZ)
   B=RotateAboutAxis(
    NormalizedRotationAxis=Normalize(VectorInput=@n8.RGB)
    RotationAngle=Multiply(A=@n10, B=@n2)
    PivotPoint=@n6.Object Pivot Location
    Position=@n5.XYZ
   )
   Alpha=ScalarParameter() {parameterName=UseSwimWPO}
  )
  n11 = DynamicParameter() {parameterIndex=2, paramNames [FresnelColorR, FresnelColorG, FresnelColorB, FresnelMax]}
  n12 = ScalarParameter() {defaultValue=5, group=Emissive, parameterName=Emissive - Fresnel Exp, sortPriority=10}
  n13 = Multiply(
   A=LinearInterpolate(
    A=ScalarParameter() {group=Emissive, parameterName=Emissive - Fresnel Min, sortPriority=11}
    B=@n11.FresnelMax
    Alpha=Fresnel(ExponentIn=@n12) {baseReflectFraction=0}
   )
   B=LinearInterpolate(
    A=VectorParameter() {parameterName=FresnelColor, defaultValue=<rgba 1 1 1 1>}.RGB
    B=@n11.RGB
    Alpha=ScalarParameter() {parameterName=UseParticleColorForFresnel}
   )
  )
  n14 = ScalarParameter() {defaultValue=1, group=Emissive, parameterName=Emissive Color Fresnel, sliderMax=1}
  MaterialFunctionCall(Emissive=Multiply(A=LinearInterpolate(A=@n13, Alpha=@n14) {constB=0}, B=ScalarParameter() {parameterName=Emissive Color})) {materialFunction=&MF_FX_EmissiveAdjustments}
  n15 = DynamicParameter() {paramNames [Dissolve, Fish Flop Speed, Fish Flop Amount, Param4]}
  NamedRerouteDeclaration(None=@n15.Fish Flop Speed) {name=Fish Flop Speed, nodeColor=<rgba 0.9688 0 1 1>}
  NamedRerouteDeclaration(None=@n15.Fish Flop Amount) {name=Fish Flop Intensity, nodeColor=<rgba 0 1 0.6484 1>}
  NamedRerouteDeclaration(None=@n15.Dissolve) {name=Dissolve, nodeColor=<rgba 0 0.1094 1 1>}
  Reroute(None=ScalarParameter() {parameterName=Metallic})
  Reroute(None=ScalarParameter() {parameterName=Specular})
  n16 = TextureSample() {texture=/Game/Effects/Fort_Effects/Textures/Tile/T_Smooth_Tile_Chromatic}
  n17 = Subtract(A=Add(A=OneMinus(None=@n16.R)), B=Multiply(A=NamedRerouteUsage()) {constB=2})
  Reroute(None=MaterialFunctionCall(Alpha Threshold=Saturate(None=@n17)) {materialFunction=&DitherTemporalAA_Fortnite}.Result)
  n18 = Constant3Vector() {constant=<rgba 1 1 1 1>}
  n19 = TextureSampleParameter2D() {
   group=Global
   parameterName=Mask
   samplerSource=SSM_Wrap_WorldGroupSettings
   texture=/Game/Gadgets/Assets/Fish/Less_Effective_Flopper_Fish/Textures/T_Floppers_M
  }
  n20 = Multiply(
   A=@n19.B
   B=TextureSample(UVs=TextureCoordinate() {coordinateIndex=1}) {texture=/Game/Gadgets/Assets/Fish/Flopper_Mark_II/Textures/T_Item_Flopper_Mark_II_Eye_D}.RGB
  )
  n21 = ParticleColor()
  n22 = Clamp(
   None=TextureSampleParameter2D() {parameterName=Diffuse Texture, texture=/Game/Gadgets/Assets/Fish/Kev_Fish/Textures/T_Item_Kev_Fish_D}.RGB
   Min=ScalarParameter() {defaultValue=0.3, parameterName=Clamp Min}
   Max=ScalarParameter() {defaultValue=1, parameterName=Clamp Max}
  ) {minDefault=0.3}
  Reroute(None=LinearInterpolate(A=Multiply(A=@n21.RGB, B=@n22), B=Multiply(A=@n18, B=@n20)))
  ScalarParameter() {defaultValue=-0.8, group=Emissive, parameterName=Emissive - Fresnel Max, sortPriority=12}
  ScalarParameter() {defaultValue=10, parameterName=Flop Intensity Debug}
  ScalarParameter() {defaultValue=1, parameterName=Flop Speed Debug}
  ScalarParameter() {parameterName=Roughness}
  TextureSampleParameter2D() {parameterName=Normal Texture, samplerType=SAMPLERTYPE_Normal, texture=/Game/Gadgets/Assets/Fish/Kev_Fish/Textures/T_Item_Kev_Fish_N_512x512}
 ]
}
